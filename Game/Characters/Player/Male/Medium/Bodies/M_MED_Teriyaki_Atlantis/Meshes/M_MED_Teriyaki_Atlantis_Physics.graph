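# Physics asset for the Teriyaki Atlantis medium male player body, set up for the dynamic loincloth bones.
# Capsules on the pelvis, thighs and calves give the body shapes, while boxes form left and right four-segment flap chains.
# The chains hang from small base bodies linked through a flap root to the pelvis with free rotation.
# Segments have tighter swing and twist limits along each chain, and cross-constraints tie matching left and right segments together.
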PhysicsAsset /Game/Characters/Player/Male/Medium/Bodies/M_MED_Teriyaki_Atlantis/Meshes/M_MED_Teriyaki_Atlantis_Physics
bodies {
 calf_l [{shapeType=Capsule, length=27.7486, radius=11.6635, center=<-27.5089 0.0065 -1.2691>, rotation=<rot 0 -113.0342 -86.2001>}]
 calf_r [{shapeType=Capsule, length=25.6223, radius=11.6056, center=<29.7919 -0.4892 1.3364>, rotation=<rot 0 -114.3355 -85.4816>}]
 dyn_loincloth_01_Base_l [{shapeType=Box, extentX=1.01, extentY=1.01, extentZ=1.01, rotation=<rot 0 -90 0>}]
 dyn_loincloth_01_Base_r [{shapeType=Box, extentX=1.01, extentY=1.01, extentZ=1.01, rotation=<rot 0 -90 0>}]
 dyn_loincloth_01_l [{shapeType=Box, extentX=2.1214, extentY=6.6514, extentZ=12.1433, center=<2.7863 -0.0014 0>, rotation=<rot 0 89.9995 -180>}]
 dyn_loincloth_01_r [{shapeType=Box, extentX=2.1214, extentY=6.6514, extentZ=13.3556, center=<2.7119 0.0005 0>, rotation=<rot 0 90 0>}]
 dyn_loincloth_02_l [{shapeType=Box, extentX=2.4385, extentY=8.0144, extentZ=12.1392, center=<4.0789 -0.0012 0>, rotation=<rot 0 89.9995 -180>}]
 dyn_loincloth_02_r [{shapeType=Box, extentX=2.4385, extentY=8.0144, extentZ=13.5641, center=<3.9912 -0.0001 0>, rotation=<rot 0 90 0>}]
 dyn_loincloth_03_l [{shapeType=Box, extentX=2.2475, extentY=8.73, extentZ=11.296, center=<4.3445 -0.0015 0>, rotation=<rot 0 89.9995 -180>}]
 dyn_loincloth_03_r [{shapeType=Box, extentX=2.2475, extentY=8.73, extentZ=13.2963, center=<4.1431 0.0001 0>, rotation=<rot 0 90 0>}]
 dyn_loincloth_04_l [{shapeType=Box, extentX=2.4649, extentY=6.7922, extentZ=6.4376, center=<3.6532 -0.0019 0>, rotation=<rot 0 89.9995 -180>}]
 dyn_loincloth_04_r [{shapeType=Box, extentX=2.4649, extentY=6.7922, extentZ=9.1775, center=<3.5707 0.0001 0>, rotation=<rot 0 90 0>}]
 dyn_loincloth_FlapRoot [{shapeType=Box, extentX=1.01, extentY=1.01, extentZ=1.01, rotation=<rot 0 -90 0>}]
 pelvis [{shapeType=Capsule, length=11.93, radius=15.907, center=<-6.9979 -2.7083 0.0731>, rotation=<rot 0.0201 84.8716 -89.8297>}]
 thigh_twist_01_l [{shapeType=Capsule, length=19.7109, radius=11.2801, center=<10 0.8517 0>, rotation=<rot 90 0 0>}]
 thigh_twist_01_r [{shapeType=Capsule, length=19.9465, radius=10.5474, center=<-10.0001 -1.5031 0>, rotation=<rot 90 0 0>}]
}
constraints [
 {bone1Name=calf_l, bone2Name=pelvis, swing1LimitDegrees=45, swing2LimitDegrees=45, twistLimitDegrees=45}
 {bone1Name=thigh_twist_01_l, bone2Name=pelvis, swing1LimitDegrees=45, swing2LimitDegrees=45, twistLimitDegrees=45}
 {bone1Name=calf_r, bone2Name=pelvis, swing1LimitDegrees=45, swing2LimitDegrees=45, twistLimitDegrees=45}
 {bone1Name=thigh_twist_01_r, bone2Name=pelvis, swing1LimitDegrees=45, swing2LimitDegrees=45, twistLimitDegrees=45}
 {
  bone1Name=dyn_loincloth_FlapRoot
  bone2Name=pelvis
  swing1LimitDegrees=45
  swing1Motion=Free
  swing2LimitDegrees=45
  swing2Motion=Free
  twistLimitDegrees=45
  twistMotion=Free
 }
 {bone1Name=dyn_loincloth_01_r, bone2Name=dyn_loincloth_01_Base_r, swing1LimitDegrees=30, swing2LimitDegrees=30, twistLimitDegrees=10}
 {bone1Name=dyn_loincloth_02_r, bone2Name=dyn_loincloth_01_r, swing1LimitDegrees=30, swing2LimitDegrees=30, twistLimitDegrees=10}
 {bone1Name=dyn_loincloth_03_r, bone2Name=dyn_loincloth_02_r, swing1LimitDegrees=30, swing2LimitDegrees=30, twistLimitDegrees=10}
 {bone1Name=dyn_loincloth_04_r, bone2Name=dyn_loincloth_03_r, swing1LimitDegrees=30, swing2LimitDegrees=30, twistLimitDegrees=10}
 {bone1Name=dyn_loincloth_01_l, bone2Name=dyn_loincloth_01_Base_l, swing1LimitDegrees=30, swing2LimitDegrees=30, twistLimitDegrees=10}
 {bone1Name=dyn_loincloth_02_l, bone2Name=dyn_loincloth_01_l, swing1LimitDegrees=30, swing2LimitDegrees=30, twistLimitDegrees=10}
 {bone1Name=dyn_loincloth_03_l, bone2Name=dyn_loincloth_02_l, swing1LimitDegrees=30, swing2LimitDegrees=30, twistLimitDegrees=10}
 {bone1Name=dyn_loincloth_04_l, bone2Name=dyn_loincloth_03_l, swing1LimitDegrees=30, swing2LimitDegrees=30, twistLimitDegrees=10}
 {bone1Name=dyn_loincloth_01_l, bone2Name=dyn_loincloth_01_r, swing1LimitDegrees=45, swing2LimitDegrees=45, twistLimitDegrees=45}
 {bone1Name=dyn_loincloth_02_l, bone2Name=dyn_loincloth_02_r, swing1LimitDegrees=45, swing2LimitDegrees=45, twistLimitDegrees=45}
 {bone1Name=dyn_loincloth_03_l, bone2Name=dyn_loincloth_03_r, swing1LimitDegrees=45, swing2LimitDegrees=45, twistLimitDegrees=45}
 {bone1Name=dyn_loincloth_04_l, bone2Name=dyn_loincloth_04_r, swing1LimitDegrees=45, swing2LimitDegrees=45, twistLimitDegrees=45}
 {
  bone1Name=dyn_loincloth_01_Base_r
  bone2Name=dyn_loincloth_FlapRoot
  swing1LimitDegrees=45
  swing1Motion=Free
  swing2LimitDegrees=45
  swing2Motion=Free
  twistLimitDegrees=45
  twistMotion=Free
 }
 {
  bone1Name=dyn_loincloth_01_Base_l
  bone2Name=dyn_loincloth_FlapRoot
  swing1LimitDegrees=45
  swing1Motion=Free
  swing2LimitDegrees=45
  swing2Motion=Free
  twistLimitDegrees=45
  twistMotion=Free
 }
]
