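# Animation Blueprint for the Teriyaki Atlantis medium male player body, following the attached parent mesh's pose and adding secondary loincloth motion.
# The copied pose runs through the body's Control Rig, then two rotation constraints blend each loincloth base bone between its opposite-side twin and same-side thigh twist.
# A RigidBody node then simulates the override physics asset in root-bone space with damped component motion and custom gravity.
# The event graph has no logic.
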AnimBlueprint /Game/Characters/Player/Male/Medium/Bodies/M_MED_Teriyaki_Atlantis/Meshes/M_MED_Teriyaki_Atlantis_AnimBP
target_skeleton=&/M_MED_Teriyaki_Atlantis_Skeleton
graphs [
 AnimGraph {nodes [
   n1 = Control Rig(Source=Copy Pose From Mesh() {bUseAttachedParent=true}) {
    alphaBoolBlend {blendOption=Linear}
    controlRigAssetReference {blueprintRigClass=/Game/Characters/Player/Male/Medium/Bodies/M_MED_Teriyaki_Atlantis/Meshes/M_MED_Teriyaki_Atlantis_CtrlRig/M_MED_Teriyaki_Atlantis_CtrlRig.M_MED_Teriyaki_Atlantis_CtrlRig_C}
   }
   n2 = "Constraint - dyn_loincloth_01_Base_r (RR)"!(ComponentPose=Local To Component(LocalPose=@n1), Alpha=1, ConstraintWeights_0=0.4, ConstraintWeights_1:float=0.6) -> (Pose) {
    alphaBoolBlend {blendOption=Linear}
    boneToModify {boneName=dyn_loincloth_01_Base_r}
    constraintSetup [{transformType=Rotation, targetBone {boneName=dyn_loincloth_01_Base_l}}, {transformType=Rotation, targetBone {boneName=thigh_twist_01_r}}]
    constraintWeights [1,1]
   }
   n3 = "Constraint - dyn_loincloth_01_Base_l (RR)"!(ComponentPose=@n2, Alpha=1, ConstraintWeights_0=0.4, ConstraintWeights_1:float=0.6) -> (Pose) {
    alphaBoolBlend {blendOption=Linear}
    boneToModify {boneName=dyn_loincloth_01_Base_l}
    constraintSetup [{transformType=Rotation, targetBone {boneName=dyn_loincloth_01_Base_r}}, {transformType=Rotation, targetBone {boneName=thigh_twist_01_l}}]
    constraintWeights [1,1]
   }
   n4 = RigidBody(ComponentPose=@n3) {
    bClampLinearTranslationLimitToRefPose=true
    bOverrideWorldGravity=true
    overridePhysicsAsset=&/M_MED_Teriyaki_Atlantis_Physics
    simulationSpace=BaseBoneSpace
    alphaBoolBlend {blendOption=Linear}
    baseBoneRef {boneName=root}
    componentAppliedLinearAccClamp=<4000 4000 4000>
    componentLinearAccScale=<0.5 0.5 0.5>
    componentLinearVelScale=<0.2 0.2 0.2>
    overrideWorldGravity {z=-980}
   }
   Output Pose(Result=Component To Local(ComponentPose=@n4)) {layerGroup=None}
  ]}
 EventGraph {nodes [Event BlueprintUpdateAnimation()]}
]
variables []
